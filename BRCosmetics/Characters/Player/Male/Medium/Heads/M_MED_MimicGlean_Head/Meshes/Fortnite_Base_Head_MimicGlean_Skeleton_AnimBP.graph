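# Animation Blueprint for the MimicGlean male head cosmetic, built on the SoupGoal female head's custom skeleton.
# The AnimGraph passes its Input Pose straight to the output with no modification.
# OnBeginTargeting and OnEndTargeting events toggle an Is aiming flag, which nothing in the graphs reads.
# The update event and its pawn owner lookup are unconnected stubs.
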AnimBlueprint /BRCosmetics/Characters/Player/Male/Medium/Heads/M_MED_MimicGlean_Head/Meshes/Fortnite_Base_Head_MimicGlean_Skeleton_AnimBP
target_skeleton=/BRCosmetics/Characters/Player/Female/Medium/Heads/F_MED_SoupGoal_Head/Meshes/CustomSkeleton/Fortnite_Base_Head_SoupGoal_Skeleton
graphs [
 AnimGraph {nodes [Output Pose(Result=Input Pose())]}
 EventGraph {nodes [
   Event OnBeginTargeting() {is=title K2Node_Event}
   > Set Is aiming(Is aiming:bool=true) -> (Output_Get:bool)
   Event OnEndTargeting() {is=title K2Node_Event}
   > Set Is aiming(Is aiming=false)
   TryGetPawnOwner()
   Event BlueprintUpdateAnimation()
  ]}
]
variables [Is aiming {type=boolean}]
